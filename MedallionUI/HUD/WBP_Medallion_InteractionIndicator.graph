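# Interaction indicator widget that shows an augment's name and description for a pickup the player can interact with.
# On context info update, it casts the receiving actor to a pickup and its item definition to a player augment, then fills both texts.
# That update also sends Event.CollectedPickup to the owning player with the augment definition as payload.
# Overrides replace the interact text with a "Pick Up" variable and enable default-widget error display.
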WidgetBlueprint /MedallionUI/HUD/WBP_Medallion_InteractionIndicator
graphs [
 EventGraph {nodes [
   n1 = Event OnContextInfoUpdated!() -> (OutputDelegate, then, ContextInfo:Fort Interact Context Info Object Reference) {is=title K2Node_Event}
   > Is Valid(InputObject=@n1.ContextInfo)
   > n2 = Cast To FortPickup!(execute, Object=Get ReceivingActor!(self:Fort Interact Context Info Object Reference=@n1.ContextInfo) -> (ReceivingActor:Actor Weak object)) -> (then, CastFailed, AsFort Pickup:Fort Pickup Object Reference) {is=title K2Node_DynamicCast}
   > n3 = Get!(execute:Exec, self:Fort Pickup Object Reference=@n2) -> (then:Exec, else:Exec, PrimaryPickupDummyItem:Fort Item Object Reference)
   > n4 = Cast To FortPlayerAugmentItemDefinition!(execute, Object=GetItemDefinition!(self:Fort Item Object Reference=@n3) -> (ReturnValue:Fort Item Definition Object Reference)) -> (then, CastFailed, AsFort Player Augment Item Definition:Fort Player Augment Item Definition Object Reference) {is=title K2Node_DynamicCast}
   > "SetText (Text)"(
    self=Get Text_AugmentName() -> (Text_AugmentName:Common Text Object Reference)
    InText=GetDisplayName!(self:Item Definition Base Object Reference=@n4) -> (ReturnValue:Text)
   )
   > "SetText (Text)"(
    self=Get Text_AugmentDescription() -> (Text_AugmentDescription:Common Text Object Reference)
    InText=GetItemShortDescription!(self:Item Definition Base Object Reference=@n4) -> (ReturnValue:Text)
   )
   > SendGameplayEventToActor!(
    execute
    Actor:Actor Object Reference=GetOwningPlayer()
    EventTag:Gameplay Tag Structure=(TagName="Event.CollectedPickup")
    Payload:Gameplay Event Data Structure=Make Gameplay Event Data!(
     EventTag:Gameplay Tag Structure=(TagName="")
     Instigator:Actor Object Reference
     Target:Actor Object Reference
     OptionalObject:Object Reference=@n4
     OptionalObject2:Object Reference
     ContextHandle:Gameplay Effect Context Handle Structure=()
     InstigatorTags:Gameplay Tag Container Structure=(GameplayTags=)
     TargetTags:Gameplay Tag Container Structure=(GameplayTags=)
     EventMagnitude:float=0
     TargetData:Gameplay Ability Target Data Handle Structure=()
    ) -> (GameplayEventData:Gameplay Event Data Structure) {is=title K2Node_MakeStruct}
   ) -> (then)
  ]}
 GetInteractText {nodes [
   GetInteractText() {is=title K2Node_FunctionEntry}
   > Parent: GetInteractText!(execute) -> (then, InteractText:Text, Result:bool) {is=title K2Node_CallParentFunction}
   > Return Node!(execute, InteractText:Text=Get InteractText() -> (InteractText:Text), Result:bool=true)
  ]}
 GetShowErrorsUsingDefaultWidget {nodes [
   GetShowErrorsUsingDefaultWidget() {is=title K2Node_FunctionEntry}
   > Return Node!(execute, ReturnValue:bool=true)
  ]}
]
variables [InteractText {type=string, default=Pick Up}]
widgets {
 info {inheritedWidgetCount=0, parentClass=/Script/FortniteGame.FortInteractContextInfoWidget, rootWidgetClass=&Overlay, widgetCount=6}
 widgets [
  Overlay_Root {widgetClassPath=&Overlay, widget=Overlay_Root}
  VerticalBox {
   widgetClassPath=&VerticalBox
   parent=Overlay_Root
   widget=VerticalBox
   slot {class=OverlaySlot, properties {padding=<ltrb 0 5 0 8>}}
  }
  HorizontalBox_0 {
   widgetClassPath=&HorizontalBox
   parent=VerticalBox
   widget=HorizontalBox_0
   properties {flowDirectionPreference=Culture}
   slot {class=VerticalBoxSlot}
  }
  Text_AugmentName {
   widgetClassPath=&CommonTextBlock
   bIsVariable=true
   parent=HorizontalBox_0
   widget=Text_AugmentName
   properties {
    flowDirectionPreference=Culture
    lineHeightPercentage=0.9
    mobileFontSizeMultiplier=1.2
    text="[PH] Augment Name"
    textTransformPolicy=ToUpper
    font {
     bForceMonospaced=false
     bMaterialIsStencil=false
     fontMaterial=null
     fontObject=&Burbank
     letterSpacing=50
     monospacedWidth=1
     size=20
     skewAmount=0
     typefaceFontName=BurbankRegular-Black
     outlineSettings {
      bApplyOutlineToDropShadows=false
      bMiteredCorners=false
      bSeparateFillAlpha=true
      outlineMaterial=null
      outlineSize=0
      outlineColor=<rgba 0 0 0 1>
     }
    }
    margin=<ltrb 0 -2 0 0>
    shadowOffset=<6 6>
   }
   slot {class=HorizontalBoxSlot}
  }
  Spacer_68 {
   widgetClassPath=&Spacer
   parent=HorizontalBox_0
   widget=Spacer_68
   slot {class=HorizontalBoxSlot, properties {size {sizeRule=Fill, value=1}}}
  }
  Text_AugmentDescription {
   widgetClassPath=&CommonTextBlock
   bIsVariable=true
   parent=VerticalBox
   widget=Text_AugmentDescription
   properties {
    lineHeightPercentage=0.9
    mobileFontSizeMultiplier=1.45
    text="[PH] Augment Description."
    wrapTextAt=432
    font {
     bForceMonospaced=false
     bMaterialIsStencil=false
     fontMaterial=null
     fontObject=&BurbankSmall
     letterSpacing=0
     monospacedWidth=1
     size=13
     skewAmount=0
     typefaceFontName=Bold
     outlineSettings {
      bApplyOutlineToDropShadows=false
      bMiteredCorners=false
      bSeparateFillAlpha=true
      outlineMaterial=null
      outlineSize=0
      outlineColor=<rgba 0 0 0 1>
     }
    }
   }
   slot {class=VerticalBoxSlot}
  }
 ]
}
