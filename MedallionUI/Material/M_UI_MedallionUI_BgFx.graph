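# Additive UI material for MedallionUI that draws an animated radial shine as a background effect.
# Two blurred-noise samples pan at different speeds over radial-converted UVs and sum into a Shine signal, using explicit UV derivatives for mip selection.
# An exponential radial gradient, set by ShineRadius and ShineAlpha, forms the circular mask.
# The color chain overlay-blends AccentColor with BaseColor and multiplies in the rerouted signals.
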Material /MedallionUI/Material/M_UI_MedallionUI_BgFx
graph {
 properties {bAutomaticallySetUsageInEditor=false, blendMode=BLEND_Additive, materialDomain=MD_UI, usedShadingModels=Default Lit}
 nodes [
  MaterialFunctionCall(Radius=Constant() {r=0.5}, Density=Constant() {r=1.5}) {materialFunction=&RadialGradientExponential}
  n1 = VectorParameter() {parameterName=AccentColor, defaultValue=<rgba 1 0.8204 0.1562 1>}
  n2 = NamedRerouteUsage()
  Multiply(
   A=MaterialFunctionCall(Base=Multiply(A=@n1.RGBA, B=@n2), Blend=VectorParameter() {parameterName=BaseColor, defaultValue=<rgba 0.8542 0.2753 0 1>}.RGB) {materialFunction=&Blend_Overlay}.Result
   B=Multiply(A=@n2, B=NamedRerouteUsage()) {constB=2}
  )
  n3 = MaterialFunctionCall() {materialFunction=&VectorToRadialValue}
  n4 = Multiply(A=Time(), B=ScalarParameter() {defaultValue=0.4, parameterName=ShineSpeed})
  n5 = Multiply(A=TextureCoordinate())
  n6 = DDX(Value=@n5)
  n7 = DDY(Value=@n5)
  n8 = TextureSample!(
   UVs=Multiply(A=Panner(Coordinate=@n3.Vector Converted to Angle, Time=@n4) {speedX=0.15}) {constB=2}
   Tex
   "DDX(UVs)"=@n6
   "DDY(UVs)"=@n7
   Apply View MipBias
  ) {mipValueMode=TMVM_Derivative, samplerType=SAMPLERTYPE_Grayscale, texture=&LowResBlurredNoise}
  n9 = TextureSample(
   UVs=Multiply(A=Panner(Coordinate=@n3.Vector Converted to Angle, Time=@n4) {speedX=-0.125, speedY=-0.125}) {constB=3}
   "DDX(UVs)"=@n6
   "DDY(UVs)"=@n7
  ) {mipValueMode=TMVM_Derivative, samplerType=SAMPLERTYPE_Grayscale, texture=&LowResBlurredNoise}
  NamedRerouteDeclaration(None=Multiply(A=Add(A=@n8.R, B=@n9.R)) {constB=1.2}) {name=Shine, nodeColor=<rgba 0 0 0 1>}
  n10 = ScalarParameter() {defaultValue=0.6, group=Shine, parameterName=ShineRadius, sortPriority=1}
  n11 = ScalarParameter() {defaultValue=1.5, group=Shine, parameterName=ShineAlpha, sortPriority=2}
  NamedRerouteDeclaration(None=Power(Base=MaterialFunctionCall(Radius=@n10, Density=@n11) {materialFunction=&RadialGradientExponential}.RadialGradientExponential)) {name=CircleMask, nodeColor=<rgba 0 0 0 1>}
 ]
}
