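# HUD widget that shows a medallion icon for a player augment item, with a glow behind it and animated show and hide.
# SetInfo stores the item, async-loads its preview into the light-wipe dynamic material, then reveals the widget; Remove plays the hide animation.
# Construct starts collapsed and sizes by playlist: small without glow for ForbiddenFruit, otherwise larger with glow.
# After loading, items lacking the MedallionExclude tag also become small and glowless.
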WidgetBlueprint /MedallionUI/HUD/WBP_Medallion_Icon
animations [
 ShowMedallion {
  binding_count=2
  length_seconds=0.35
  object=&self:ShowMedallion
  sequence {bindings [
    Image_MedallionIcon {
     id=#1
     possessed_class=FortLazyImage
     tracks [
      {
       type=MovieScene2DTransformTrack
       sections [{
         type=MovieScene2DTransformSection
         end_frame=7
         start_frame=0
         channels [
          Translation.X {type=MovieSceneScriptingFloatChannel, const [0,2]}
          Translation.Y {type=MovieSceneScriptingFloatChannel, keys [[0,10],[5,0]]}
          Scale.X {type=MovieSceneScriptingFloatChannel, keys [[0,1],[3,1.3],[7,1]]}
          Scale.Y {type=MovieSceneScriptingFloatChannel, keys [[0,1],[3,1.3],[7,1]]}
         ]
        }]
      }
      {
       type=MovieSceneFloatTrack
       display_name=Render Opacity
       sections [{type=MovieSceneFloatSection, end_frame=7, start_frame=0, channels [None {type=MovieSceneScriptingFloatChannel, keys [[0,0],[7,1]]}]}]
      }
     ]
    }
    Image_Glow {
     id=#2
     possessed_class=Image
     tracks [
      {
       type=MovieSceneFloatTrack
       display_name=Render Opacity
       sections [{type=MovieSceneFloatSection, end_frame=7, start_frame=0, channels [None {type=MovieSceneScriptingFloatChannel, keys [[0,0],[7,1]]}]}]
      }
      {
       type=MovieScene2DTransformTrack
       sections [{
         type=MovieScene2DTransformSection
         end_frame=7
         start_frame=0
         channels [
          Translation.X {type=MovieSceneScriptingFloatChannel, const [0,2]}
          Translation.Y {type=MovieSceneScriptingFloatChannel, keys [[0,10],[5,0]]}
          Scale.X {type=MovieSceneScriptingFloatChannel, keys [[0,1],[3,1.3],[7,1]]}
          Scale.Y {type=MovieSceneScriptingFloatChannel, keys [[0,1],[3,1.3],[7,1]]}
         ]
        }]
      }
     ]
    }
   ]}
 }
 HideMedallion {
  binding_count=2
  length_seconds=0.35
  object=&self:HideMedallion
  sequence {bindings [
    Image_Glow {
     id=#3
     possessed_class=Image
     tracks [
      {
       type=MovieSceneFloatTrack
       display_name=Render Opacity
       sections [{type=MovieSceneFloatSection, end_frame=7, start_frame=0, channels [None {type=MovieSceneScriptingFloatChannel, keys [[0,1],[5,0]]}]}]
      }
      {
       type=MovieScene2DTransformTrack
       sections [{
         type=MovieScene2DTransformSection
         end_frame=7
         start_frame=0
         channels [
          Translation.X {type=MovieSceneScriptingFloatChannel, const [0,1]}
          Translation.Y {type=MovieSceneScriptingFloatChannel, keys [[0,10],[7,20]]}
         ]
        }]
      }
     ]
    }
    Image_MedallionIcon {
     id=#4
     possessed_class=FortLazyImage
     tracks [
      {
       type=MovieScene2DTransformTrack
       sections [{
         type=MovieScene2DTransformSection
         end_frame=7
         start_frame=0
         channels [
          Translation.X {type=MovieSceneScriptingFloatChannel, const [0,1]}
          Translation.Y {type=MovieSceneScriptingFloatChannel, keys [[0,10],[7,20]]}
         ]
        }]
      }
      {
       type=MovieSceneFloatTrack
       display_name=Render Opacity
       sections [{type=MovieSceneFloatSection, end_frame=7, start_frame=0, channels [None {type=MovieSceneScriptingFloatChannel, keys [[0,1],[5,0]]}]}]
      }
     ]
    }
   ]}
 }
]
graphs [
 EventGraph {nodes [
   Event Remove() {is=title K2Node_CustomEvent}
   > PlayAnimation(InAnimation=Get HideMedallion() -> (HideMedallion:Widget Animation Object Reference))
   n1 = Event SetInfo!() -> (OutputDelegate, then, InItemDef:Fort Player Augment Item Definition Object Reference) {is=title K2Node_CustomEvent}
   > n2 = Set ItemDef(ItemDef:Fort Player Augment Item Definition Object Reference=@n1.InItemDef) -> (Output_Get:Fort Player Augment Item Definition Object Reference)
   > n3 = Async Load Asset(Asset=GetSmallPreviewImage!(ItemDefinition:Item Definition Base Object Reference=@n2) -> (ReturnValue:Texture 2D Soft Object Reference))
   n4 = Cast To Texture2D(execute=@n3.Completed, Object=@n3)
   > n5 = GetDynamicMaterial(self=Get Image_MedallionIcon() -> (Image_MedallionIcon:Fort Lazy Image Object Reference))
   > SetTextureParameterValue(self=@n5, ParameterName=MedallionIconTexture, Value=@n4)
   > SetVisibility(InVisibility=HitTestInvisible)
   > PlayAnimation(InAnimation=Get ShowMedallion() -> (ShowMedallion:Widget Animation Object Reference))
   > n6 = Branch(Condition=HasTag!(
     TagContainer:"Gameplay Tag Container Structure (by ref)"=GetOwnedGameplayTags!(TagContainerInterface:Gameplay Tag Asset Interface Interface=@n2) -> (ReturnValue:Gameplay Tag Container Structure)
     Tag:Gameplay Tag Structure=(TagName="HUD.Athena.Inventory.MedallionExclude")
     bExactMatch:bool=false
    ) -> (ReturnValue:bool))
   n7 = Get SizeBox() -> (SizeBox:Size Box Object Reference)
   SetHeightOverride(execute=@n6.else, self=@n7, InHeightOverride=28)
   > SetWidthOverride(self=@n7, InWidthOverride=28)
   > SetVisibility(self=Get Image_Glow() -> (Image_Glow:Image Object Reference), InVisibility=Collapsed)
   Event Construct()
   > SetVisibility(InVisibility=Collapsed)
   > SetIconSize()
  ]}
 GetItemDef {nodes [
   GetItemDef() {is=title K2Node_FunctionEntry}
   > Return Node!(
    execute
    MedallionItemDef:Fort Player Augment Item Definition Object Reference=Get ItemDef() -> (ItemDef:Fort Player Augment Item Definition Object Reference)
   )
  ]}
 SetIconSize {nodes [
   SetIconSize() {is=title K2Node_FunctionEntry}
   > n1 = Cast To FortGameStateAthena!(execute, Object=GetGameState!() -> (ReturnValue:Game State Base Object Reference)) -> (then, CastFailed, AsFort Game State Athena:Fort Game State Athena Object Reference) {is=title K2Node_DynamicCast}
   > n2 = Branch(Condition=PlaylistHasTag!(
     self:Fort Game State Athena Object Reference=@n1
     InTag:"Gameplay Tag Structure (by ref)"=MakeLiteralGameplayTag!(Value:Gameplay Tag Structure=(TagName="Athena.Playlist.ForbiddenFruit")) -> (ReturnValue:Gameplay Tag Structure)
    ) -> (ReturnValue:bool))
   n3 = Get SizeBox()
   > SetHeightOverride(self=@n3, InHeightOverride=28)
   > SetWidthOverride(self=@n3, InWidthOverride=28)
   > SetVisibility(self=Get Image_Glow(), InVisibility=Collapsed)
   n4 = Get SizeBox()
   SetHeightOverride(execute=@n2.else, self=@n4, InHeightOverride=45)
   > SetWidthOverride(self=@n4, InWidthOverride=45)
   > SetVisibility(self=Get Image_Glow(), InVisibility=SelfHitTestInvisible)
  ]}
]
variables [
 GameplayTag {type=GameplayTag, default {tagName=None}}
 ItemDef {type=FortPlayerAugmentItemDefinition, default=null}
]
widgets {
 info {inheritedWidgetCount=0, parentClass=&UserWidget, rootWidgetClass=&Overlay, widgetCount=4}
 widgets [
  Overlay_0 {widgetClassPath=&Overlay, widget=Overlay_0}
  Image_Glow {
   widgetClassPath=&Image
   bIsVariable=true
   parent=Overlay_0
   widget=Image_Glow
   properties {
    visibility=Collapsed
    brush {
     bIsDynamicallyLoaded=false
     drawAs=Image
     imageType=FullColor
     mirroring=NoMirror
     resourceName=None
     resourceObject=/MedallionUI/Material/M_UI_MedallionUI_BgFx
     tiling=NoTile
     imageSize=<70 70>
     margin=<ltrb 0 0 0 0>
     outlineSettings {
      bUseBrushTransparency=false
      roundingType=HalfHeightRadius
      width=0
      color {colorUseRule=UseColor_Specified, specifiedColor=<rgba 0 0 0 0>}
      cornerRadii=<0 0 0 0>
     }
     tintColor {colorUseRule=UseColor_Specified, specifiedColor=<rgba 1 1 1 1>}
     uVRegion {bIsValid=false, max=<0 0>, min=<0 0>}
    }
   }
   slot {class=OverlaySlot, properties {horizontalAlignment=HAlign_Fill, verticalAlignment=VAlign_Fill}}
  }
  SizeBox {
   widgetClassPath=&SizeBox
   bIsVariable=true
   parent=Overlay_0
   widget=SizeBox
   properties {bOverride_HeightOverride=true, bOverride_WidthOverride=true, heightOverride=28, widthOverride=28}
   slot {class=OverlaySlot, properties {horizontalAlignment=HAlign_Center, verticalAlignment=VAlign_Center}}
  }
  Image_MedallionIcon {
   widgetClassPath=&FortLazyImage
   bIsVariable=true
   parent=SizeBox
   widget=Image_MedallionIcon
   properties {brush {
     bIsDynamicallyLoaded=false
     drawAs=Image
     imageType=FullColor
     mirroring=NoMirror
     resourceName=None
     resourceObject=/MedallionUI/Material/MI_UI_MedallionIcon_LightWipe
     tiling=NoTile
     imageSize=<28 28>
     margin=<ltrb 0 0 0 0>
     outlineSettings {
      bUseBrushTransparency=false
      roundingType=HalfHeightRadius
      width=0
      color {colorUseRule=UseColor_Specified, specifiedColor=<rgba 0 0 0 0>}
      cornerRadii=<0 0 0 0>
     }
     tintColor {colorUseRule=UseColor_Specified, specifiedColor=<rgba 1 1 1 1>}
     uVRegion {bIsValid=false, max=<0 0>, min=<0 0>}
    }}
   slot {class=SizeBoxSlot}
  }
 ]
}
