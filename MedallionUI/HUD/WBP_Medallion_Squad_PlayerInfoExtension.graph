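# Squad HUD player info extension widget that shows a player's active medallion and challenge quest icons in two rows.
# OnMedallionTagChanged binds to the player's augment list updates and sorts active augments by gameplay tag into medallions and challenges.
# Refresh diffs existing icons against those lists, removing stale ones, adding only missing ones and collapsing empty rows.
# Medallion icons are also removed on refresh while the player is dead; PreConstruct adds design-time preview icons.
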WidgetBlueprint /MedallionUI/HUD/WBP_Medallion_Squad_PlayerInfoExtension
graphs [
 AddNewChallengeQuestWidgets {nodes [
   AddNewChallengeQuestWidgets() {is=title K2Node_FunctionEntry}
   > n1 = For Each Loop(Array:Array of Fort Player Augment Item Definition Object References=Get NewChallenges() -> (NewChallenges:Array of Fort Player Augment Item Definition Object References)) -> (Array Element:Fort Player Augment Item Definition Object Reference)
   > n2 = Branch(Condition=Contains!(
     TargetMap:Map of Fort Player Augment Item Definition Object References to WBP Medallion Icon Object References=Get Temp_PopulatedChallenges() -> (Temp_PopulatedChallenges:Map of Fort Player Augment Item Definition Object References to WBP Medallion Icon Object References)
     Key:"Fort Player Augment Item Definition Object Reference (by ref)"=@n1
    ) -> (ReturnValue))
   n3 = Create Widget!(
    execute:Exec=@n2.else
    Class:User Widget Class Reference=&/WBP_Medallion_Icon.WBP_Medallion_Icon_C
    OwningPlayer:Player Controller Object Reference
   ) -> (then:Exec, ReturnValue:WBP Medallion Icon Object Reference)
   > AddChildToHorizontalBox!(
    execute
    self:Horizontal Box Object Reference=Get HorizontalBox_ChallengeQuests() -> (HorizontalBox_ChallengeQuests:Horizontal Box Object Reference)
    Content:Widget Object Reference=@n3
   ) -> (then, ReturnValue:Horizontal Box Slot Object Reference)
   > Event SetInfo!(execute, self:WBP Medallion Icon Object Reference=@n3, InItemDef:Fort Player Augment Item Definition Object Reference=@n1) -> (then)
   Return Node(execute=@n1.Completed)
  ]}
 AddNewMedallionWidgets {nodes [
   AddNewMedallionWidgets() {is=title K2Node_FunctionEntry}
   > n1 = For Each Loop(Array=Get NewMedallions() -> (NewMedallions:Array of Fort Player Augment Item Definition Object References))
   > n2 = Branch(Condition=Contains(
     TargetMap=Get Temp_PopulatedMedallions() -> (Temp_PopulatedMedallions:Map of Fort Player Augment Item Definition Object References to WBP Medallion Icon Object References)
     Key=@n1
    ))
   n3 = Create Widget(execute=@n2.else, Class=&WBP_Medallion_Icon_C)
   > n4 = AddChildToHorizontalBox(self=Get HorizontalBox_MedallionIcons() -> (HorizontalBox_MedallionIcons:Horizontal Box Object Reference), Content=@n3)
   > Set Padding!(execute, Padding:Margin Structure=Make Margin(Left=-7, Top=0, Right=-7, Bottom=0), self:Horizontal Box Slot Object Reference=@n4) -> (then, Output_Get:Margin Structure)
   > Event SetInfo(self=@n3, InItemDef=@n1)
   Return Node(execute=@n1.Completed)
  ]}
 Clear Widgets {nodes [
   Clear Widgets() {is=title K2Node_FunctionEntry}
   > ClearChildren(self=[Get HorizontalBox_ChallengeQuests(), Get HorizontalBox_MedallionIcons()])
   > SetVisibility(self=[Get HorizontalBox_MedallionIcons(), Get HorizontalBox_ChallengeQuests()], InVisibility=Collapsed)
   > Clear(TargetArray:Array of Fort Player Augment Item Definition Object References=Get NewMedallions())
   > Clear!(
    execute
    TargetMap:Map of Fort Player Augment Item Definition Object References to WBP Medallion Icon Object References=Get Temp_PopulatedMedallions()
   ) -> (then) {is=title K2Node_CallFunction}
   > Clear(TargetArray=Get NewChallenges()) {is=title K2Node_CallArrayFunction}
   > Clear(TargetMap=Get Temp_PopulatedChallenges()) {is=title K2Node_CallFunction}
  ]}
 EventGraph {nodes [
   n1 = Event OnMedallionTagChanged!() -> (OutputDelegate, then, InPlayerState:Fort Player State Athena Object Reference) {is=title K2Node_Event}
   > n2 = Is Valid(InputObject=@n1.InPlayerState)
   > Set Player State Ref(Player State Ref:Fort Player State Athena Object Reference=@n1.InPlayerState) -> (Output_Get:Fort Player State Athena Object Reference)
   > Bind Event to On Augment List Updated(
    self:Fort Player State Component Player Augment System Object Reference=GetAugmentComponent!(PlayerContextActor:Actor Object Reference=Get Player State Ref() -> (Player State Ref:Fort Player State Athena Object Reference)) -> (ReturnValue:Fort Player State Component Player Augment System Object Reference)
    Delegate=Create Event()
   ) {is=title K2Node_AddDelegate}
   > UpdateActiveMedallionsAndChallenges!(
    execute
    self
    AugmentComponent:Fort Player State Component Player Augment System Object Reference=GetAugmentComponent(PlayerContextActor=Get Player State Ref())
   ) -> (then)
   SetVisibility(execute=@n2.Is Not Valid, self=[Get HorizontalBox_MedallionIcons(), Get HorizontalBox_ChallengeQuests()], InVisibility=Collapsed)
   Event Construct()
   > Clear Widgets() {is=title K2Node_CallFunction}
   n3 = Event PreConstruct()
   > Branch(Condition=@n3.IsDesignTime)
   > n4 = For Loop(LastIndex=3)
   n5 = Create Widget(execute=@n4.Completed, Class=&WBP_Medallion_Icon_C)
   > AddChildToHorizontalBox(self=Get HorizontalBox_ChallengeQuests(), Content=@n5)
   n6 = Create Widget(execute=@n4.LoopBody, Class=&WBP_Medallion_Icon_C)
   > n7 = AddChildToHorizontalBox(self=Get HorizontalBox_MedallionIcons(), Content=@n6)
   > Set Padding(Padding=Make Margin(Left=-7, Top=0, Right=-7, Bottom=0), self=@n7)
   n8 = Event OnPlayerDead!() -> (OutputDelegate, then, PSA:Fort Player State Athena Object Reference, bIsDead:bool) {is=title K2Node_Event}
   > Set IsAlive(IsAlive:bool=NOT Boolean(A=@n8.bIsDead)) -> (Output_Get:bool)
   Get NewMedallions()
   Get NewChallenges()
  ]}
 GetOwnedGameplayTagBugWorkaround {nodes [
   n1 = GetOwnedGameplayTagBugWorkaround!() -> (then, GameplayTagAssetInterface:Gameplay Tag Asset Interface Interface) {is=title K2Node_FunctionEntry}
   > Return Node!(
    execute
    TagContainer:Gameplay Tag Container Structure=Get Owned Gameplay Tags!(self:Gameplay Tag Asset Interface Interface=@n1) -> (ReturnValue:Gameplay Tag Container Structure) {is=title K2Node_CallFunction}
   )
  ]}
 RefreshPopulatedWidgets {nodes [
   RefreshPopulatedWidgets() {is=title K2Node_FunctionEntry}
   > Clear(TargetMap=Get Temp_PopulatedMedallions())
   > Clear(TargetMap=Get Temp_PopulatedChallenges())
   > n1 = Set SavedChildren(SavedChildren:Array of Widget Object References=GetAllChildren(self=Get HorizontalBox_MedallionIcons())) -> (Output_Get:Array of Widget Object References)
   > n2 = For Each Loop(Array:Array of Widget Object References=@n1) -> (Array Element:Widget Object Reference)
   > n3 = Cast To WBP_Medallion_Icon!(execute, Object=@n2) -> (then, CastFailed, AsWBP Medallion Icon:WBP Medallion Icon Object Reference) {is=title K2Node_DynamicCast}
   n4 = GetItemDef!(self:WBP Medallion Icon Object Reference=@n3) -> (MedallionItemDef:Fort Player Augment Item Definition Object Reference)
   n5 = Get NewMedallions()
   > n6 = Branch(Condition=AND Boolean(
     A=Contains Item(
      TargetArray:Array of Fort Player Augment Item Definition Object References=@n5
      ItemToFind:"Fort Player Augment Item Definition Object Reference (by ref)"=@n4
     )
     B=Get IsAlive() -> (IsAlive:bool)
    ))
   > Add!(
    execute
    TargetMap:Map of Fort Player Augment Item Definition Object References to WBP Medallion Icon Object References=Get Temp_PopulatedMedallions()
    Key:"Fort Player Augment Item Definition Object Reference (by ref)"=@n4
    Value:"WBP Medallion Icon Object Reference (by ref)"=@n3
   ) -> (then) {is=title K2Node_CallFunction}
   Event Remove(execute=@n6.else, self:WBP Medallion Icon Object Reference=@n3)
   > RemoveChild(self=Get HorizontalBox_MedallionIcons(), Content=@n2)
   n7 = Set SavedChildren(execute=@n2.Completed, SavedChildren=GetAllChildren(self=Get HorizontalBox_ChallengeQuests()))
   > n8 = For Each Loop(Array=@n7)
   Return Node!(execute=@n8.Completed)
   > n9 = Cast To WBP_Medallion_Icon(Object=@n8)
   n10 = GetItemDef(self=@n9)
   > n11 = Branch(Condition=Contains Item(TargetArray=Get NewChallenges(), ItemToFind=@n10))
   > Add(TargetMap=Get Temp_PopulatedChallenges(), Key=@n10, Value=@n9)
   Event Remove(execute=@n11.else, self=@n9)
   > RemoveChild(self=Get HorizontalBox_ChallengeQuests(), Content=@n8)
  ]}
 UpdateActiveMedallionsAndChallenges {nodes [
   UpdateActiveMedallionsAndChallenges!() -> (then, AugmentComponent:Fort Player State Component Player Augment System Object Reference) {is=title K2Node_FunctionEntry}
   > Clear(TargetArray=Get NewMedallions()) {is=title K2Node_CallArrayFunction}
   > Clear(TargetArray=Get NewChallenges())
   > n1 = GetActiveAugments!(
    execute
    Actor:Actor Object Reference=Get Player State Ref()
    None:%=Get AugmentComponent() -> (AugmentComponent:Fort Player State Component Player Augment System Object Reference)
   ) -> (then, ReturnValue:Array of Fort Player Augment Item Definition Object References)
   > n2 = For Each Loop(Array:Array of Fort Player Augment Item Definition Object References=@n1) -> (Array Element:Fort Player Augment Item Definition Object Reference)
   n3 = GetOwnedGameplayTagBugWorkaround!(self, GameplayTagAssetInterface:Gameplay Tag Asset Interface Interface=@n2) -> (TagContainer:Gameplay Tag Container Structure) {is=title K2Node_CallFunction}
   > n4 = Branch(Condition=HasTag!(
     TagContainer:"Gameplay Tag Container Structure (by ref)"=@n3
     Tag:Gameplay Tag Structure=(TagName="Augment.Category.Medallion")
     bExactMatch:bool=true
    ) -> (ReturnValue:bool))
   > Add(
    TargetArray:Array of Fort Player Augment Item Definition Object References=Get NewMedallions()
    NewItem:"Fort Player Augment Item Definition Object Reference (by ref)"=@n2
   ) {is=title K2Node_CallArrayFunction}
   Branch(execute=@n4.else, Condition=HasTag(TagContainer=@n3, Tag=(TagName="Augment.Category.ChallengeQuest"), bExactMatch=true))
   > Add(TargetArray=Get NewChallenges(), NewItem=@n2)
   RefreshPopulatedWidgets(execute=@n2.Completed) {is=title K2Node_CallFunction}
   > AddNewMedallionWidgets() {is=title K2Node_CallFunction}
   > AddNewChallengeQuestWidgets() {is=title K2Node_CallFunction}
   n5 = Get HorizontalBox_MedallionIcons()
   n6 = GetChildrenCount!(self:Panel Widget Object Reference=@n5) -> (ReturnValue:int=0)
   > SetVisibility(self=@n5, InVisibility=Select(Option 0=Collapsed, Option 1=SelfHitTestInvisible, Index="integer > integer"(A=@n6)))
   n7 = Get HorizontalBox_ChallengeQuests()
   n8 = GetChildrenCount(self=@n7) -> (ReturnValue=0)
   > SetVisibility(self=@n7, InVisibility=Select(Option 0=Collapsed, Option 1=SelfHitTestInvisible, Index="integer > integer"(A=@n8)))
  ]}
]
variables [
 IsAlive {type=boolean, default=true}
 NewChallenges {type="array<FortPlayerAugmentItemDefinition>", default []}
 NewMedallions {type="array<FortPlayerAugmentItemDefinition>", default []}
 Player State Ref {type=FortPlayerStateAthena, default=null}
 Temp_PopulatedChallenges {default {}}
 Temp_PopulatedMedallions {default {}}
]
widgets {
 info {inheritedWidgetCount=1, parentClass=/Script/MedallionUI.MedallionsInfoWidget, rootWidgetClass=&Overlay, widgetCount=6}
 widgets [
  Overlay_17 {widgetClassPath=&Overlay, widget=Overlay_17}
  ScaleBox_0 {
   widgetClassPath=&ScaleBox
   parent=Overlay_17
   widget=ScaleBox_0
   properties {stretch=UserSpecified}
   slot {class=OverlaySlot, properties {verticalAlignment=VAlign_Center}}
  }
  HorizontalBox_0 {widgetClassPath=&HorizontalBox, parent=ScaleBox_0, widget=HorizontalBox_0, slot {class=ScaleBoxSlot}}
  HorizontalBox_MedallionIcons {
   widgetClassPath=&HorizontalBox
   bIsVariable=true
   parent=HorizontalBox_0
   widget=HorizontalBox_MedallionIcons
   slot {class=HorizontalBoxSlot, properties {horizontalAlignment=HAlign_Left, verticalAlignment=VAlign_Center, padding=<ltrb 7 0 0 0>}}
  }
  HorizontalBox_ChallengeQuests {
   widgetClassPath=&HorizontalBox
   bIsVariable=true
   parent=HorizontalBox_0
   widget=HorizontalBox_ChallengeQuests
   slot {class=HorizontalBoxSlot, properties {horizontalAlignment=HAlign_Left, verticalAlignment=VAlign_Center, padding=<ltrb 7 0 0 0>}}
  }
  TutorialHighlight {widgetClassPath=&TutorialHighlight}
 ]
}
